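# Sound cue for the electric medallion's spawn sound, a spatialized loot ambient that pairs a one-shot start with a sustained loop.
# A mixer layers the start wave with the looping wave, whose enveloper fades it in over the first 0.7 seconds under the start.
# A modulator at the root randomizes pitch, and the cue uses ambient item loop attenuation with a concurrency override capping simultaneous instances at two.
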SoundCue /Game/Sounds/Foley_Loot/Medallions/Medallion_Elec_Spawn_Cue
root=Modulator_0
nodes [
 {
  id=Enveloper_0
  children [WavePlayer_2]
  properties {pitchCurve {editorCurveData {keys [{time=0}]}}, volumeCurve {editorCurveData {keys [<at 0 0>,<at 0.35 0.1>,<at 0.7 1>]}}}
 }
 {id=Mixer_0, children [WavePlayer_1,Enveloper_0], properties {inputVolume [1,1]}}
 {id=Modulator_0, children [Mixer_0], properties {pitchMax=1.2, pitchMin=0.9, volumeMax=1, volumeMin=1}}
 {id=WavePlayer_1, properties {soundWaveAssetPtr=&/Medallion_Spawn_Elec_Start_01}}
 {id=WavePlayer_2, properties {bLooping=true, soundWaveAssetPtr=&/Medallion_Spawn_Elec_Loop_01}}
]
properties {
 attenuationSettings=/Game/Sounds/Attenuation/Fort_Item_Ambient_Loop_Mono_Default_Att
 bOverrideConcurrency=true
 maxDistance=1250
 soundClassObject=/Game/Sounds/FortSoundClassesAndModes/Fort_Loot_Ambient_Loop
 volumeMultiplier=0.7
 concurrencyOverrides {maxCount=2}
}
