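# Sound cue for the sound of a medallion loot item landing, classed as a loot collect sound.
# A single wave plays through a modulator that randomizes pitch slightly while keeping volume fixed.
# The cue plays at reduced volume with weapon-drop attenuation and its own concurrency override that caps simultaneous instances.
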SoundCue /Game/Sounds/Foley_Loot/Medallions/Medallion_Land_Cue
root=Modulator_0
nodes [
 {id=Modulator_0, children [WavePlayer_0], properties {pitchMax=1.2, pitchMin=0.9, volumeMax=1, volumeMin=1}}
 {id=WavePlayer_0, properties {soundWaveAssetPtr=&/Medallion_Land_01}}
]
properties {
 attenuationSettings=&Fort_Weapon_Drop_Att
 bOverrideConcurrency=true
 maxDistance=3000
 soundClassObject=&Fort_Loot_Collect
 volumeMultiplier=0.5
 concurrencyOverrides {maxCount=2}
}
